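# Sound cue for the reactive sound of the Conscience Red character cosmetic, classed as player foley.
# A Random node picks one of three equally weighted reactive waves and feeds a root Modulator for slight variation.
# Foley attenuation spatializes it with boosted volume, and a concurrency override allows one instance at a time, stopping the oldest.
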SoundCue /BRCosmetics/Sounds/Characters/Conscience_Red/Char_ConscienceRed_Reactive_Cue
root=Modulator_0
nodes [
 {id=Modulator_0, children [Random_0], properties {volumeMin=1}}
 {id=Random_0, children [WavePlayer_6,WavePlayer_7,WavePlayer_8], properties {weights [1,1,1]}}
 {id=WavePlayer_6, properties {soundWaveAssetPtr=&/Char_ConscienceRed_Reactive_03}}
 {id=WavePlayer_7, properties {soundWaveAssetPtr=&/Char_ConscienceRed_Reactive_01}}
 {id=WavePlayer_8, properties {soundWaveAssetPtr=&/Char_ConscienceRed_Reactive_02}}
]
properties {
 attenuationSettings=&Foley_Attenuation_Basic_Att
 bOverrideConcurrency=true
 maxDistance=1536
 soundClassObject=&Fort_Player_Foley
 volumeMultiplier=2
 concurrencyOverrides {maxCount=1, resolutionRule=StopOldest}
}
